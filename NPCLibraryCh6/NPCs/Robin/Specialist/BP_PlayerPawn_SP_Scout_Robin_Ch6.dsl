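; Player pawn Blueprint for the Chapter 6 Robin scout specialist NPC, adding reactive speech bubbles once the NPC is hired.
; When the pawn is converted, it plays Robin's OnHire bubble and starts listening through the EventProcessorSubsystem for events on the converting pawn that match a verb filter.
; Each event that passes the filter triggers the OnEvent speech bubble on Robin.
; BeginPlay, overlap and Tick are disabled stubs.
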(event EventBeginPlay
  (|CallParentFunction) ; This node is disabled and will not be called.\nDrag off pins to build functionality.
)

(event Collision|EventActorBeginOverlap
  (params (OtherActor :type Actor%20Object%20Reference))
  (|CallParentFunction OtherActor) ; This node is disabled and will not be called.\nDrag off pins to build functionality.
)

(event EventTick
  (params (DeltaSeconds :type Float%20%28single-precision%29))
  (|CallParentFunction DeltaSeconds) ; This node is disabled and will not be called.\nDrag off pins to build functionality.
)

(event Conversion|EventHandlePawnConverted
  (params (InstigatorPawn :type Fort%20Pawn%20Object%20Reference) (ConvertedPawn :type Fort%20Pawn%20Object%20Reference))
  (|CallParentFunction InstigatorPawn ConvertedPawn)
  (Class|BPFLAICommonFunctions|PlaySpeechBubble ConvertedPawn "(TagName=\"Conversation.Entry.Robin.SpeechBubble.OnHire\")")
  (CallFunction|OnEvent InstigatorPawn))

(event Custom|OnEvent
  (params (Converter :type Pawn%20Object%20Reference))
  (EventProcessorSubsystem|StartListeningToFilteredEvent 0 Converter (Variables|Default|GetVerb%20Filter) false (EventDispatchers|CreateEvent :!event Custom|OnFilterPassed)))

(event Custom|OnFilterPassed
  (Class|BPFLAICommonFunctions|PlaySpeechBubble self "(TagName=\"Conversation.Entry.Robin.SpeechBubble.OnEvent\")" false))

(fn ConstructionScript
  (|CallParentFunction))
